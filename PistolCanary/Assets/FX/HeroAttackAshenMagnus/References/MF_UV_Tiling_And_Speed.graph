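# MF_UV_Tiling_And_Speed is a material function in the HeroAttackAshenMagnus FX folder that rotates, tiles, pans and offsets UVs.
# CustomRotator turns the input UV, TexCoord by default, by an angle before it is multiplied by tiling.
# A time-scaled speed and the Nois_Tex offset are then added.
# Tiling uses only R of UV_TIle_Speed and speed only B, each duplicated across both axes, leaving G and A unused.
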MaterialFunction /PistolCanary/Assets/FX/HeroAttackAshenMagnus/References/MF_UV_Tiling_And_Speed
graph {nodes [
  n1 = TextureCoordinate()
  n2 = MaterialFunctionCall(Float4=FunctionInput() {bUsePreviewValueAsDefault=true, inputName=UV_TIle_Speed, inputType=FunctionInput_Vector4, sortPriority=2}) {materialFunction=&BreakOutFloat4Components}
  n3 = Multiply(
   A=MaterialFunctionCall(
    UVs=FunctionInput(Preview=@n1) {bUsePreviewValueAsDefault=true, inputName=Rotation_UV, inputType=FunctionInput_Vector2}
    "Rotation Angle (0-1)"=FunctionInput() {bUsePreviewValueAsDefault=true, inputName=Rotate_Angle, inputType=FunctionInput_Scalar}
   ) {materialFunction=&CustomRotator}.Rotated Values
   B=AppendVector(A=@n2.R, B=@n2.R)
  )
  n4 = Multiply(
   A=AppendVector(A=@n2.B, B=@n2.B)
   B=FunctionInput(Preview=Time()) {bUsePreviewValueAsDefault=true, inputName=Time, inputType=FunctionInput_Scalar, sortPriority=3}
  )
  n5 = FunctionInput() {bUsePreviewValueAsDefault=true, inputName=Nois_Tex, inputType=FunctionInput_Vector2, sortPriority=4, previewValue=<1 0 0 1>}
  FunctionOutput(None=Add(A=Add(A=@n3, B=@n4), B=@n5))
 ]}
