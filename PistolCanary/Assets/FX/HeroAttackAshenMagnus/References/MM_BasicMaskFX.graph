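# Masked, unlit, two-sided Niagara sprite material for Ashen Magnus hero attack FX, defaulting to the "Sounds" texture.
# A dynamic Frame parameter picks a cell from a 2x2 flipbook; red times particle color, divided by eye adaptation, gives exposure-independent color.
# Green times particle alpha forms the mask, optionally temporal-dithered except at low quality.
# Tiled, scrolling cloud noise, offset by dynamic Dissolve and smoothstepped by Soft_Edge, is subtracted to erode it.
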Material /PistolCanary/Assets/FX/HeroAttackAshenMagnus/References/MM_BasicMaskFX
graph {
 properties {
  bUsedWithNiagaraSprites=true
  blendMode=BLEND_Masked
  pixelDepthOffsetMode=PDOM_Legacy
  shadingModel=MSM_Unlit
  twoSided=true
  usedShadingModels=Unlit
 }
 nodes [
  n1 = ParticleColor()
  n2 = Constant2Vector() {g=2, r=2}
  n3 = TextureSampleParameter2D(UVs=MaterialFunctionCall!(
    U_Images=@n2.R
    V_Images=@n2.R
    SubUV Texture RGB
    Time Speed
    Manual Time Control=StaticBool() {value=true}
    Time=DynamicParameter() {paramNames [Dissolve,Frame,Param3,Param4]}.Frame
    UVs
   ) {materialFunction=/Game/Effects/Fort_Effects/Materials/_Functions/MF_SubUV}.UVs) {parameterName=Main Texture, texture=&/T_PistolCanary_AshenMagnus_Sounds}
  Divide(A=Multiply(A=@n1.RGB, B=StaticSwitchParameter(True=@n3.R, False=@n3.R) {parameterName=RIsEmissive}), B=EyeAdaptation().EyeAdaptation)
  n4 = Multiply(A=@n1.A, B=StaticSwitchParameter(True=@n3.G, False=@n3.G) {defaultValue=true, parameterName=GIsAlpha})
  n5 = MaterialFunctionCall(Alpha Threshold=@n4) {materialFunction=&DitherTemporalAA}
  n6 = ScalarParameter() {defaultValue=0.6, group=Dissolve, parameterName=Soft_Edge, sliderMax=1, sliderMin=0.5}
  n7 = VectorParameter() {group=Dissolve, parameterName=T_Dissolve_Tile_Speed, defaultValue=<rgba 1 1 0 -0.2>}
  n8 = TextureSampleParameter2D(UVs=MaterialFunctionCall!(Rotate_Angle, Rotation_UV, UV_TIle_Speed=AppendVector(A=@n7.RGB, B=@n7.RGB), Time, Nois_Tex) {materialFunction=&/MF_UV_Tiling_And_Speed}.Result) {
   group=Dissolve
   parameterName=T_Dissolve
   samplerType=SAMPLERTYPE_LinearColor
   texture=&/T_CloudNoise_003
  }
  n9 = Subtract(
   A=Add(A=StaticComponentMaskParameter(None=@n8.RGB) {defaultR=true, parameterName=RGB})
   B=Multiply(A=DynamicParameter() {paramNames [Dissolve,Frame,Param3,Param4]}.Dissolve) {constB=3}
  )
  Subtract(
   A=QualitySwitch(Default=StaticSwitchParameter(True=@n5.Result, False=@n4) {defaultValue=true, parameterName=EnableDitherAA}, Low=@n4)
   B=SmoothStep(Min=OneMinus(None=@n6), Max=@n6, Value=Saturate(None=@n9))
  )
 ]
}
